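# Animation Blueprint for the Skyfire exterior gadget's doors, opening and closing the center door and five small doors A–E from booleans.
# DynamicDoors layers timed pose blends per fan bone for the small doors, and the AnimGraph layers the center door over it, with a custom curve for Fan_center_a.
# Non-dynamic mode switches one-way to StaticOpen or StaticClosed, both with unconnected outputs.
# SetCenterOpenInstant zeroes blend times, sets the center state, then restores times after a delay.
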AnimBlueprint /Game/Gadgets/Assets/Skyfire_Exterior_Gadget/Meshes/Skyfire_Doors_AnimBP
target_skeleton=&/Skyfire_Exterior_Gadget_Skeleton
graphs [
 AnimGraph {nodes [
   n1 = Blend Poses by bool(
    BlendPose_0=Sequence Player 'Skyfire_MainDoor_Open'() {sequence=/Game/Animation/Game/Gadgets/Skyfire_Exterior/Skyfire_MainDoor_Open}
    BlendTime_0=Get currentOpenTime_BottomInterior() -> (currentOpenTime_BottomInterior:double)
    BlendTime_1=Get currentCloseTime_BottomInterior() -> (currentCloseTime_BottomInterior:double)
    bActiveValue=Get DoorsOpen_Center() -> (DoorsOpen_Center:bool)
   ) {blendType=Custom, customBlendCurve=/Game/Gadgets/Assets/Skyfire_Exterior_Gadget/Meshes/Anim/SkyfireCenterSmallDoorOpen_Curve}
   n2 = Blend Poses by bool(
    BlendPose_0=Sequence Player 'Skyfire_MainDoor_Open'() {sequence=&Skyfire_MainDoor_Open}
    BlendTime_0=Get currentOpenTime() -> (currentOpenTime:double)
    BlendTime_1=Get currentCloseTime() -> (currentCloseTime:double)
    bActiveValue=Get DoorsOpen_Center()
   ) {blendType=Sinusoidal}
   AnimGraphNode_StateMachine_0 = DoorStates() {is=title AnimGraphNode_StateMachine}
   Output Pose(Result=Layered blend per bone(
     BasePose=@AnimGraphNode_StateMachine_0
     BlendPoses_0=Layered blend per bone(BasePose=@n2, BlendPoses_0=@n1) {layerSetup [{branchFilters [{boneName=Fan_center_a}]}]}
    ) {layerSetup [{branchFilters [{boneName=Fan_center}, {boneName=Fan_center_a}]}]}) {layerGroup=None}
  ]}
 DoorStates {
  entry=DynamicDoors
  path=AnimGraph.AnimGraphNode_StateMachine_0.DoorStates
  ids=id
  nodes [
   AnimStateNode_0 = DynamicDoors() {is=title AnimStateNode}
   AnimStateNode_1 = StaticClosed() {is=title AnimStateNode}
   AnimStateNode_2 = StaticOpen() {is=title AnimStateNode}
   AnimStateTransitionNode_4 = AnimStateTransitionNode() {crossfadeDuration=0}
   AnimStateTransitionNode_5 = AnimStateTransitionNode() {crossfadeDuration=0}
  ]
  links [
   {
    dst=StaticClosed
    rule=AnimGraph.AnimGraphNode_StateMachine_0.DoorStates.AnimStateTransitionNode_5.Transition
    src=DynamicDoors
    via=AnimStateTransitionNode_5
   }
   {
    dst=StaticOpen
    rule=AnimGraph.AnimGraphNode_StateMachine_0.DoorStates.AnimStateTransitionNode_4.Transition
    src=DynamicDoors
    via=AnimStateTransitionNode_4
   }
  ]
 }
 DynamicDoors {
  path=AnimGraph.AnimGraphNode_StateMachine_0.DoorStates.AnimStateNode_0.DynamicDoors
  nodes [
   n1 = Sequence Player 'Skyfire_MainDoor_Open'() {sequence=&Skyfire_MainDoor_Open}
   n2 = Get currentOpenTime_BottomInterior()
   n3 = Get currentCloseTime_BottomInterior()
   n4 = Get DoorsOpen_Center()
   n5 = Sequence Player 'Skyfire_MainDoor_Open'() {sequence=&Skyfire_MainDoor_Open}
   n6 = Get currentOpenTime()
   n7 = Get currentCloseTime()
   n8 = Get DoorsOpen_Center()
   Layered blend per bone(BlendPoses_0=Layered blend per bone(
     BasePose=Blend Poses by bool(BlendPose_0=@n5, BlendTime_0=@n6, BlendTime_1=@n7, bActiveValue=@n8) {blendType=Sinusoidal}
     BlendPoses_0=Blend Poses by bool(BlendPose_0=@n1, BlendTime_0=@n2, BlendTime_1=@n3, bActiveValue=@n4) {blendType=Custom, customBlendCurve=&SkyfireCenterSmallDoorOpen_Curve}
    ) {layerSetup [{branchFilters [{boneName=Fan_center_a}]}]}) {layerSetup [{branchFilters [{boneName=Fan_center}, {boneName=Fan_center_a}]}]}
   n9 = Sequence Player 'Skyfire_SmallDoors_Open'() {sequence=/Game/Animation/Game/Gadgets/Skyfire_Exterior/Skyfire_SmallDoors_Open}
   n10 = Get DoorsOpen_A() -> (DoorsOpen_A:bool)
   n11 = Get currentOpenTime()
   n12 = Get currentCloseTime()
   n13 = Blend Poses by bool(
    BlendPose_0=Sequence Player 'Skyfire_SmallDoors_Open'() {sequence=&Skyfire_SmallDoors_Open}
    BlendTime_0=Get currentOpenTime()
    BlendTime_1=Get currentCloseTime()
    bActiveValue=Get DoorsOpen_B() -> (DoorsOpen_B:bool)
   ) {blendType=Sinusoidal}
   n14 = Sequence Player 'Skyfire_SmallDoors_Open'() {sequence=&Skyfire_SmallDoors_Open}
   n15 = Get currentOpenTime()
   n16 = Get currentCloseTime()
   n17 = Get DoorsOpen_D() -> (DoorsOpen_D:bool)
   n18 = Blend Poses by bool(
    BlendPose_0=Sequence Player 'Skyfire_SmallDoors_Open'() {sequence=&Skyfire_SmallDoors_Open}
    BlendTime_0=Get currentOpenTime()
    BlendTime_1=Get currentCloseTime()
    bActiveValue=Get DoorsOpen_E() -> (DoorsOpen_E:bool)
   ) {blendType=Sinusoidal}
   n19 = Layered blend per bone(
    BasePose=Layered blend per bone(
     BasePose=Layered blend per bone(BlendPoses_0=@n18) {layerSetup [{branchFilters [{boneName=Fan_e}, {boneName=Fan_e_e}]}]}
     BlendPoses_0=Blend Poses by bool(BlendPose_0=@n14, BlendTime_0=@n15, BlendTime_1=@n16, bActiveValue=@n17) {blendType=Sinusoidal}
    ) {layerSetup [{branchFilters [{boneName=Fan_d}, {boneName=Fan_d_d}]}]}
    BlendPoses_0=Blend Poses by bool(
     BlendPose_0=Sequence Player 'Skyfire_SmallDoors_Open'() {sequence=&Skyfire_SmallDoors_Open}
     BlendTime_0=Get currentOpenTime()
     BlendTime_1=Get currentCloseTime()
     bActiveValue=Get DoorsOpen_C() -> (DoorsOpen_C:bool)
    ) {blendType=Sinusoidal}
   ) {layerSetup [{branchFilters [{boneName=Fan_c}, {boneName=Fan_c_c}]}]}
   Output Animation Pose(Result=Layered blend per bone(
     BasePose=Layered blend per bone(BasePose=@n19, BlendPoses_0=@n13) {layerSetup [{branchFilters [{boneName=Fan_b}, {boneName=Fan_b_b}]}]}
     BlendPoses_0=Blend Poses by bool(BlendPose_0=@n9, BlendTime_0=@n11, BlendTime_1=@n12, bActiveValue=@n10) {blendType=Sinusoidal}
    ) {layerSetup [{branchFilters [{boneName=Fan_a}, {boneName=Fan_a_a}]}]})
  ]
 }
 StaticClosed {path=AnimGraph.AnimGraphNode_StateMachine_0.DoorStates.AnimStateNode_1.StaticClosed, nodes [Output Animation Pose()]}
 StaticOpen {
  path=AnimGraph.AnimGraphNode_StateMachine_0.DoorStates.AnimStateNode_2.StaticOpen
  nodes [
   Layered blend per bone(BlendPoses_0=Sequence Player 'Skyfire_MainDoor_Open'() {sequence=&Skyfire_MainDoor_Open}) {layerSetup [{branchFilters [{boneName=Fan_center_a}, {boneName=Fan_center}]}]}
   Output Animation Pose()
  ]
 }
 Transition {
  path=AnimGraph.AnimGraphNode_StateMachine_0.DoorStates.AnimStateTransitionNode_4.Transition
  nodes [
   n1 = Get isDynamic() -> (isDynamic:bool)
   Result(bCanEnterTransition=AND Boolean(A=NOT Boolean(A=@n1), B=Get DoorsOpen_Center()))
  ]
 }
 Transition {
  path=AnimGraph.AnimGraphNode_StateMachine_0.DoorStates.AnimStateTransitionNode_5.Transition
  nodes [
   n1 = Get DoorsOpen_Center()
   n2 = Get isDynamic()
   Result(bCanEnterTransition=AND Boolean(A=NOT Boolean(A=@n2), B=NOT Boolean(A=@n1)))
  ]
 }
 EventGraph {nodes [
   n1 = SetCenterOpenInstant!() -> (OutputDelegate, then, open:bool=~) {is=title K2Node_CustomEvent}
   > n2 = Sequence!(execute) -> (then_0, then_1, then_2:Exec)
   > Set currentOpenTime(currentOpenTime:double=0) -> (Output_Get:double)
   > Set currentCloseTime(currentCloseTime:double=0) -> (Output_Get:double)
   > Set currentOpenTime_BottomInterior(currentOpenTime_BottomInterior:double=0) -> (Output_Get:double)
   > Set currentCloseTime_BottomInterior(currentCloseTime_BottomInterior:double=0) -> (Output_Get:double)
   Set DoorsOpen_Center(execute=@n2.then_1, DoorsOpen_Center:bool=@n1.open) -> (Output_Get:bool)
   Delay(execute=@n2.then_2, Duration=0.2)
   > Set currentOpenTime(currentOpenTime=Get OpenTime() -> (OpenTime:double))
   > Set currentCloseTime(currentCloseTime=Get CloseTime() -> (CloseTime:double))
   > Set currentOpenTime_BottomInterior(currentOpenTime_BottomInterior=Get OpenTime_Interior() -> (OpenTime_Interior:double))
   > Set currentCloseTime_BottomInterior(currentCloseTime_BottomInterior=Get CloseTime_Interior() -> (CloseTime_Interior:double))
   Event BlueprintBeginPlay() {is=title K2Node_Event}
   > Set currentOpenTime(currentOpenTime=Get OpenTime())
   > Set currentCloseTime(currentCloseTime=Get CloseTime())
   > Set currentOpenTime_BottomInterior(currentOpenTime_BottomInterior=Get OpenTime_Interior())
   > Set currentCloseTime_BottomInterior(currentCloseTime_BottomInterior=Get CloseTime_Interior())
  ]}
]
variables [
 CloseTime {type=number, default=6.6}
 CloseTime_Interior {type=number, default=5}
 currentCloseTime {type=number, default=6.6}
 currentCloseTime_BottomInterior {type=number, default=5}
 currentOpenTime {type=number, default=6.6}
 currentOpenTime_BottomInterior {type=number, default=7.6}
 DoorsOpen_A {type=boolean}
 DoorsOpen_B {type=boolean}
 DoorsOpen_C {type=boolean}
 DoorsOpen_Center {type=boolean}
 DoorsOpen_D {type=boolean}
 DoorsOpen_E {type=boolean}
 isDoingInstant {type=boolean}
 isDynamic {type=boolean, default=true}
 OpenTime {type=number, default=6.6}
 OpenTime_Interior {type=number, default=7.6}
]
